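# Looping daytime ambience cue for the Tennis Club POI, routed to the Athena outdoor ambient sound class.
# A quality node feeds two branches, each mixing a full-volume Industrial quad bed with a quieter Boreal Tundra bed.
# The first two quality slots use the quad tundra wave, the last two the stereo one.
# Nested endlessly looping envelopers add a slow volume dip, a fixed pitch scale and a shorter cyclic pitch drift.
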SoundCue /Game/Sounds/Helios/POI/TennisClub/Ambience_TennisClub_Day_Loop_Cue
root=QualityLevel_0
nodes [
 {
  id=Enveloper_0
  children [Enveloper_3]
  properties {
   bLoop=true
   bLoopIndefinitely=true
   loopEnd=100
   pitchCurve {editorCurveData {keys [<at 0 0.8>]}}
   volumeCurve {editorCurveData {keys [<at 0 1 RCIM_Cubic>,<at 50 0.9 RCIM_Cubic>,<at 100 1 RCIM_Cubic>]}}
  }
 }
 {
  id=Enveloper_3
  children [Mixer_0]
  properties {
   bLoop=true
   bLoopIndefinitely=true
   loopEnd=30
   pitchCurve {editorCurveData {keys [{interpMode=RCIM_Cubic, time=0}, <at 15.9468 0.9181 RCIM_Cubic>, {interpMode=RCIM_Cubic, time=30}]}}
   volumeCurve {editorCurveData {keys [<at 0 1>]}}
  }
 }
 {
  id=Enveloper_5
  children [Enveloper_6]
  properties {
   bLoop=true
   bLoopIndefinitely=true
   loopEnd=100
   pitchCurve {editorCurveData {keys [<at 0 0.8>]}}
   volumeCurve {editorCurveData {keys [<at 0 1 RCIM_Cubic>,<at 50 0.9 RCIM_Cubic>,<at 100 1 RCIM_Cubic>]}}
  }
 }
 {
  id=Enveloper_6
  children [Mixer_1]
  properties {
   bLoop=true
   bLoopIndefinitely=true
   loopEnd=30
   pitchCurve {editorCurveData {keys [{interpMode=RCIM_Cubic, time=0}, <at 15.9468 0.9181 RCIM_Cubic>, {interpMode=RCIM_Cubic, time=30}]}}
   volumeCurve {editorCurveData {keys [<at 0 1>]}}
  }
 }
 {id=Mixer_0, children [WavePlayer_5,WavePlayer_3], properties {inputVolume [0.3,1]}}
 {id=Mixer_1, children [WavePlayer_8, WavePlayer_10], properties {inputVolume [1,0.3]}}
 {id=QualityLevel_0, children [Enveloper_0,Enveloper_0,Enveloper_5,Enveloper_5]}
 {id=WavePlayer_10, properties {soundWaveAssetPtr=/Game/Sounds/Helios/Biomes/BorealTundra/Ambience_Boreal_Tundra_Day_Ext_Loop_Stereo}}
 {id=WavePlayer_3, properties {bLooping=true, soundWaveAssetPtr=/Game/Sounds/Asteria/POI/IndustrialTown/Ambience_Industrial_Day_02_Quad}}
 {id=WavePlayer_5, properties {soundWaveAssetPtr=/Game/Sounds/Helios/Biomes/BorealTundra/Ambience_Boreal_Tundra_Day_Ext_Loop_Quad}}
 {id=WavePlayer_8, properties {bLooping=true, soundWaveAssetPtr=&Ambience_Industrial_Day_02_Quad}}
]
properties {maxDistance=2097152, soundClassObject=&Fort_Ambient_Outdoors_Athena, volumeMultiplier=0.401}
