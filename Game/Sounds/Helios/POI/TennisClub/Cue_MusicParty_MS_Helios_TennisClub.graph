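# Sound cue that plays looping diegetic music for the Tennis Club POI in Helios.
# A looping wave player feeds an Enveloper that ramps volume from silence to full over the first second.
# Base submix and submix sends are off, so the cue outputs through a full-level send to the MusicParty Tennis Club audio bus.
# It overrides concurrency with a small voice cap and has a very large max distance.
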SoundCue /Game/Sounds/Helios/POI/TennisClub/Cue_MusicParty_MS_Helios_TennisClub
root=Enveloper_0
nodes [
 {
  id=Enveloper_0
  children [WavePlayer_1]
  properties {pitchCurve {editorCurveData {keys [{time=0}]}}, volumeCurve {editorCurveData {keys [<at 0 0>,<at 1 1>]}}}
 }
 {id=WavePlayer_1, properties {bLooping=true, soundWaveAssetPtr=&/s28_DiegeticMusic_TennisClub_BMG_Loop}}
]
properties {
 bEnableBaseSubmix=false
 bEnableSubmixSends=false
 bOverrideConcurrency=true
 maxDistance=2097152
 volumeMultiplier=0.55
 busSends [{audioBus=&/MusicParty_MS_Helios_TennisClub, minSendLevel=0, sendLevel=1, soundSourceBus=null}]
 concurrencyOverrides {maxCount=4}
}
